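# Translucent, unlit, two-sided mesh-particle material, by its name for loot container ribbon effects.
# Particle color feeds emissive; particle alpha drives a noise-based dissolve, multiplied by U and V gradient edge masks for opacity.
# The V mask uses noise-distorted UVs, offset by dynamic parameter Param2 to slide it along the mesh.
# Time-scrolled, world-aligned noise displaces vertices along their normals, scaled by dynamic parameter WPO_power.
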Material /Game/Effects/Fort_Effects/Effects/LootContainers/M_Loot_Ribbons_02unlit
graph {
 properties {
  bTangentSpaceNormal=false
  bUsedWithMeshParticles=true
  blendMode=BLEND_Translucent
  opacityMaskClipValue=0.999
  pixelDepthOffsetMode=PDOM_Legacy
  shadingModel=MSM_Unlit
  twoSided=true
  usedShadingModels=Unlit
 }
 nodes [
  n1 = ParticleColor()
  MaterialFunctionCall(Emissive=@n1.RGB) {materialFunction=&MF_FX_EmissiveAdjustments}
  n2 = TextureSample(UVs=TextureCoordinate() {uTiling=0.25, vTiling=0.25}) {texture=/Game/Global/Textures/Default/Tileable_Patterns/T_Tilling_Distortion_Map_01}
  n3 = Power(Base=@n2.G, Exp=LinearInterpolate(Alpha=@n1.A) {constA=5, constAlpha=0})
  n4 = LinearInterpolate(Alpha=@n1.A) {constAlpha=0, constB=250}
  n5 = Add(A=Multiply(A=ComponentMask(None=@n2.RGB) {g=true, r=true}) {constB=0.15}, B=TextureCoordinate())
  n6 = Constant() {r=-2}
  n7 = DynamicParameter() {defaultValue=<rgba 0 0 0 1>, paramNames [WPO_power,Param2,Param3,Param4]}
  n8 = MaterialFunctionCall(UV Channel=Add(A=Panner(Coordinate=@n5, Time=@n6) {speedX=0.05, speedY=0.05}, B=@n7.Param2) {constB=0, desc=0 to -1 to pan Mask}) {materialFunction=&LinearGradient}
  n9 = Clamp(None=OneMinus(None=@n8.VGradient))
  n10 = Clamp(None=LinearInterpolate(Alpha=Multiply(A=@n9, B=@n8.VGradient)))
  n11 = MaterialFunctionCall() {materialFunction=&LinearGradient}
  n12 = Clamp(None=OneMinus(None=@n11.UGradient))
  n13 = Multiply(A=Power(Base=Multiply(A=@n12, B=@n11.UGradient))) {constB=35}
  Multiply(A=Clamp(None=Multiply(A=@n3, B=@n4)), B=Multiply(A=Multiply(A=@n10) {constB=5}, B=Clamp(None=@n13)))
  n14 = TextureObject() {texture=&T_Tilling_Distortion_Map_01}
  n15 = Constant() {r=700}
  n16 = StaticBool() {value=true}
  n17 = Constant4Vector() {constant=<rgba 0 1 0 0>}
  n18 = TextureObject() {samplerType=SAMPLERTYPE_Normal, texture=/Game/Global/Textures/Default/Tileable_Patterns/T_Tilling_Distortion_Map_01_n}
  n19 = Constant3Vector() {constant=<rgba 8 8 8 0>}
  n20 = MaterialFunctionCall(Speed=Constant() {r=-0.0428}) {materialFunction=&TimeWithSpeedVariable}
  n21 = Add(
   A=Add(A=WorldPosition() {worldPositionShaderOffset=WPT_ExcludeAllShaderOffsets}.XYZ, B=Multiply(A=@n15, B=@n19))
   B=Multiply(A=Multiply(A=@n15, B=@n20.Time), B=Constant3Vector())
  )
  n22 = Normalize(VectorInput=VertexNormalWS())
  n23 = VertexNormalWS()
  Multiply(
   A=ComponentMask(None=MaterialFunctionCall!(
     Diffuse Texture Object=@n14
     Spec Texture Object
     TextureSize=@n15
     --------------------------
     Use seperate settings for Z projection
     Use seperate texture for Z projection
     Texture Size Z projection
     Z Plane Diffuse Texture Object
     Z Plane Spec Texture Object
     "--------------------------      "
     ProjectionTransitionContrast
     Use Texture Alpha Heightmap to Refine the Transistion
     HeightLerpTransitionContrast
     PickTheDiffuseHeightChannel=@n16
     TheDiffuseHeightChannel=@n17.RGBA
     "--------------------------       "
     NormalTextureObject=@n18
     World Space Normals
     "--------------------------  "
     Export Float 4 Diffuse
     Export Float 4 Spec
     "--------------------------     "
     WorldPosition=@n21
     World Space Normal=@n22
    ) {materialFunction=/Engine/Functions/Engine_MaterialFunctions02/Texturing/WorldAlignedTextures_Complex}.XYZ Diffuse Texture) {g=true}
   B=Multiply(A=Normalize(VectorInput=@n23), B=@n7.WPO_power) {constB=60}
  )
  Power(Base=@n10)
  Sine(None=Time())
 ]
}
